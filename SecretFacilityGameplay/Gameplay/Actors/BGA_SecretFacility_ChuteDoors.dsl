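; Chute door actor for the Secret Facility, with two doors that move apart when the OpenDoors custom event runs.
; BeginPlay caches each door's relative location as its starting point.
; A timeline lerps both doors, with sweep, from those points to fixed, opposite relative Z positions.
; Every timeline update also executes a burst gameplay cue for door opening at the actor's location; no closing logic is present.
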(event EventBeginPlay
  (Variables|Default|SetDoorAStartingPos (Variables|/Script/Engine.SceneComponent|GetRelativeLocation (Variables|Default|GetDoorA)))
  (Variables|Default|SetDoorBStartingPos (Variables|/Script/Engine.SceneComponent|GetRelativeLocation (Variables|Default|GetDoorB))))

(event Custom|OpenDoors
  (bind (_direction _newtrack_0) (Timeline|DoorOpeningTimeline Play)
    (:Update
      (Transformation|SetRelativeLocation (Variables|Default|GetDoorA) (Math|Vector|Lerp(Vector) (Variables|Default|GetDoorAStartingPos) "0.000000,0.000000,1100.000000" _newtrack_0) true false)
      (Transformation|SetRelativeLocation (Variables|Default|GetDoorB) (Math|Vector|Lerp(Vector) (Variables|Default|GetDoorBStartingPos) "0.000000,0.000000,-1100.000000" _newtrack_0) true false)
      (GameplayCue|ExecuteGameplayCueOnActor(Burst) self (Variables|Default|GetGCDoorsOpening) (Ability|GameplayCue|MakeGameplayCueParameters :NormalizedMagnitude 0.0 :RawMagnitude 0.0 :Location (Transformation|GetActorLocation self) :Normal "0,0,0" :GameplayEffectLevel 1 :AbilityLevel 1 :bReplicateLocationWhenUsingMinimalRepProxy false)))))

(fn ConstructionScript)
