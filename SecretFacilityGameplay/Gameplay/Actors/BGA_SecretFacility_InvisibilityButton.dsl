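; Interactable button actor for the Secret Facility that releases gas, named as an invisibility button.
; Interacting locks the button, bursts a warning GameplayCue and, one second later, activates the ReleaseGas ability granted at BeginPlay.
; A ScalableFloat-driven timer re-enables interaction.
; The construction script offsets and scales the gas collision box; ApplyInvisibility only makes it query-only and is not called here, while actor overlaps just print debug text.
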(event EventBeginPlay
  (GameplayAbilities|GiveAbility (Variables|Default|GetAbilitySystem) "/SecretFacilityGameplay/Gameplay/Abilities/GA_SecretFacility_ReleaseGas.GA_SecretFacility_ReleaseGas_C" 0 -1))

(event Interaction|EventBlueprintOnInteract
  (params (InteractingPawn :type Fort%20Pawn%20Object%20Reference) (InteractionBeingAttempted :type EInteractionBeingAttempted%20Enum :default "FirstInteraction"))
  (Variables|Default|SetAllowInteraction false)
  (Utilities|Time|SetTimerbyEvent (delegate Custom|EnableButton) (ScalableFloat|GetValueatLevel (Variables|Default|GetHFTimeUntilReactivated)))
  (GameplayCue|ExecuteGameplayCueOnActor(Burst) self (Variables|Default|GetGCActivationWarning) (Ability|GameplayCue|MakeGameplayCueParameters :NormalizedMagnitude 0.0 :RawMagnitude 0.0 :Location (Transformation|GetActorLocation self) :Normal "0,0,0" :GameplayEffectLevel 1 :AbilityLevel 1 :bReplicateLocationWhenUsingMinimalRepProxy false))
  (Utilities|Time|SetTimerbyEvent (delegate Custom|WarningComplete) 1.0))

(event Custom|WarningComplete
  (Abilities|TryActivateAbilitybyClass (Variables|Default|GetAbilitySystem) "/SecretFacilityGameplay/Gameplay/Abilities/GA_SecretFacility_ReleaseGas.GA_SecretFacility_ReleaseGas_C" true))

(event Custom|EnableButton
  (Variables|Default|SetAllowInteraction true)
  (Variables|Interaction|SetbAllowInteract true))

(event Custom|ApplyInvisibility
  (Collision|SetCollisionEnabled (Variables|Default|GetGasCollisionBox) "QueryOnly"))

(event Collision|EventActorBeginOverlap
  (params (OtherActor :type Actor%20Object%20Reference))
  (Development|PrintString (Utilities|String|Append (Utilities|String|ToString(Object) OtherActor) " overlapped!") true true "(R=1.000000,G=0.190046,B=0.000000,A=1.000000)"))

(fn BlueprintCanInteract
  (params (InteractingPawn :type Fort%20Pawn%20Object%20Reference) (InteractionBeingAttempted :type EInteractionBeingAttempted%20Enum :default "FirstInteraction") (InteractionType :type TInteractionType%20Enum :default "IT_NoInteraction"))
  (return (Variables|Default|GetAllowInteraction)))

(fn BlueprintGetInteractionString
  (params (InteractingPawn :type Fort%20Pawn%20Object%20Reference) (InteractionBeingAttempted :type EInteractionBeingAttempted%20Enum :default "FirstInteraction"))
  (return (Variables|Default|GetInteractionName)))

(fn BlueprintGetInteractionTime
  (params (InteractingPawn :type Fort%20Pawn%20Object%20Reference) (InteractionBeingAttempted :type EInteractionBeingAttempted%20Enum :default "FirstInteraction"))
  (return 1.0 true))

(fn ConstructionScript
  (Transformation|AddLocalOffset (Variables|Default|GetGasCollisionBox) (Variables|Default|GetCollisionBoxOffset) false false)
  (Transformation|SetRelativeScale3D (Variables|Default|GetGasCollisionBox) (Variables|Default|GetCollisionBoxScale)))
