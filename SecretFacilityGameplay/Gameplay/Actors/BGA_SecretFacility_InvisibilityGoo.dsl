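; Invisibility goo patch actor for Secret Facility that grants temporary invisibility to pawns overlapping its box, gated by a scalable-float toggle.
; On authority, begin overlap applies invisibility and tag effects, a looping timer refreshes invisibility, and end overlap refreshes it while removing the standing-on-goo tag.
; Bound damage handlers strip invisibility unless the pawn is DBNO.
; Construction script configures box and decal; deprecated v2 events remain.
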(event EventBeginPlay
  ;[0.07846 0.15 0.0455 0.5] v3 - Active
  ;;[1 1 1] Begin Play -> Set overlapping timer to apply invis duration refresh to pawns standing on invis
  (if (ScalableFloat|GetValueasBool (Variables|Default|GetHF_EnableInvisibilityFunctionality))
    (Utilities|Time|SetTimerbyEvent (delegate Custom|CheckOverlaps) (ScalableFloat|GetValueatLevel (Variables|Default|GetHF_InvisibilityOverlapTimerInterval)) true)))

(event Custom|CheckOverlaps
  ;[0.07846 0.15 0.0455 0.5] v3 - Active
  ;;[1 1 1] Begin Play -> Set overlapping timer to apply invis duration refresh to pawns standing on invis
  (for _array_element (Collision|GetOverlappingActors (Variables|Default|GetCollisionBox) "/Script/FortniteGame.FortPawn" :!class /Script/Engine.PrimitiveComponent)
    (bind _asfort_pawn (Utilities|Casting|CastToFortPawn _array_element))
    (if (Networking|HasAuthority)
      (GameplayEffects|ApplyGameplayEffectToSelf :self (Ability|GetAbilitySystemComponent _asfort_pawn) :GameplayEffectClass "/SecretFacilityGameplay/Gameplay/GameplayEffects/GE_SecretFacility_GrantInvisibilityTemporary.GE_SecretFacility_GrantInvisibilityTemporary_C" :Level 0.0)
      (EventDispatchers|BindEventtoOnDamaged _asfort_pawn (delegate Custom|OnInvisPawnDamaged)))))

(event Custom|OnInvisPawnDamaged
  (params (DamagedActor :type Actor%20Object%20Reference) (Damage :type Float%20%28single-precision%29) (InstigatedBy :type Controller%20Object%20Reference) (DamageCauser :type Actor%20Object%20Reference) (HitLocation :type Vector) (FHitComponent :type Primitive%20Component%20Object%20Reference) (BoneName :type Name) (Momentum :type Vector))
  (bind _returnvalue (Ability|GetFortAbilitySystemComponent DamagedActor))
  ;[0.07846 0.15 0.0455 0.5] v3 - Active
  ;;[1 1 1] On Pawn Damaged Event -> Remove invis and unbind event
  (bind _asfort_pawn (Utilities|Casting|CastToFortPawn DamagedActor))
  (if (and (not (GameplayTags|HasMatchingGameplayTag _returnvalue "(TagName=\"Gameplay.Status.DBNO\")")) (GameplayTags|HasMatchingGameplayTag _returnvalue "(TagName=\"SecretFacility.Invisibility.Active\")"))
    (GameplayEffects|RemoveActiveEffectswithGrantedTags _returnvalue "(GameplayTags=((TagName=\"SecretFacility.Invisibility.Active\")))")))

(event Collision|EventActorBeginOverlap
  (params (OtherActor :type Actor%20Object%20Reference))
  ;[0.07846 0.15 0.0455 0.5] v3 - Active
  ;;[1 1 1] On Overlap -> Apply invis to pawn and on damage event
  (if (ScalableFloat|GetValueasBool (Variables|Default|GetHF_EnableInvisibilityFunctionality))
    (bind _asfort_pawn (Utilities|Casting|CastToFortPawn OtherActor))
    (if (Networking|HasAuthority)
      (GameplayEffects|ApplyGameplayEffectToSelf :self (Ability|GetAbilitySystemComponent _asfort_pawn) :GameplayEffectClass "/SecretFacilityGameplay/Gameplay/GameplayEffects/GE_SecretFacility_GrantInvisibilityTemporary.GE_SecretFacility_GrantInvisibilityTemporary_C" :Level 0.0)
      (EventDispatchers|AssignOnDamaged _asfort_pawn (delegate Custom|OnInvisPawnDamaged))
      (GameplayEffects|ApplyGameplayEffectToSelf :self (Ability|GetAbilitySystemComponent _asfort_pawn) :GameplayEffectClass "/SecretFacilityGameplay/Gameplay/GameplayEffects/GE_SecretFacility_GrantInvisibilityTemporaryTag.GE_SecretFacility_GrantInvisibilityTemporaryTag_C" :Level 0.0))))

(event Collision|EventActorEndOverlap
  (params (OtherActor :type Actor%20Object%20Reference))
  ;[0.07846 0.15 0.0455 0.5] v3 - Active
  ;;[1 1 1] On End -> Refresh invis duration and bind damage event
  (if (ScalableFloat|GetValueasBool (Variables|Default|GetHF_EnableInvisibilityFunctionality))
    (bind _asfort_pawn (Utilities|Casting|CastToFortPawn OtherActor))
    (if (Networking|HasAuthority)
      (GameplayEffects|ApplyGameplayEffectToSelf :self (Ability|GetAbilitySystemComponent _asfort_pawn) :GameplayEffectClass "/SecretFacilityGameplay/Gameplay/GameplayEffects/GE_SecretFacility_GrantInvisibilityTemporary.GE_SecretFacility_GrantInvisibilityTemporary_C" :Level 0.0)
      (EventDispatchers|AssignOnDamaged _asfort_pawn (delegate Custom|OnInvisPawnDamaged))
      (GameplayEffects|RemoveActiveEffectswithGrantedTags (Ability|GetAbilitySystemComponent _asfort_pawn) "(GameplayTags=((TagName=\"SecretFacility.Invisibility.StandingOnGoo\")))"))))

(event Custom|CustomEvent_2
  (params (OtherActor :type Actor%20Object%20Reference)))

(event Custom|CustomEvent
  (bind _returnvalue (Ability|GetFortAbilitySystemComponent 0))
  (bind _activegootimermap (Variables|Default|GetActiveGooTimerMap))
  (bind (_value _returnvalue_2) (Utilities|Map|Find _activegootimermap _returnvalue))
  ;[0.06273 0.03051 0.15 0.5] v2 - DEPRECATED
  ;;[1 1 1] If we're still overlapping on the timer elapsed, and don't have invis then reapply
  (if (and (not (GameplayTags|HasMatchingGameplayTag 0 "(TagName=\"SecretFacility.Invisibility.Active\")")) (Collision|IsOverlappingActor 0))
    ;;[0.113 0.15 0.009115 0.5] Apply invis, gets removed on a timer later on. If a player is actively awaiting their timer and re-enters goo, we invalidate the timer and clear the map value.
    (GameplayEffects|ApplyGameplayEffectToSelf _returnvalue "/SecretFacilityGameplay/Gameplay/GameplayEffects/GE_SecretFacility_GrantInvisibilityInfinite.GE_SecretFacility_GrantInvisibilityInfinite_C" 1.0 (GameplayEffects|MakeEffectContext _returnvalue))
    (Development|PrintString "Apply invis")
    (if _returnvalue_2
      (Utilities|Time|ClearandInvalidateTimerbyHandle _value)
      (Utilities|Map|Remove _activegootimermap _returnvalue))))

(event Custom|OnDamaged_Event
  (params (DamagedActor :type Actor%20Object%20Reference) (Damage :type Float%20%28single-precision%29) (InstigatedBy :type Controller%20Object%20Reference) (DamageCauser :type Actor%20Object%20Reference) (HitLocation :type Vector) (FHitComponent :type Primitive%20Component%20Object%20Reference) (BoneName :type Name) (Momentum :type Vector))
  ;[0.06273 0.03051 0.15 0.5] v2 - DEPRECATED
  (GameplayEffects|RemoveActiveEffectswithGrantedTags (Ability|GetFortAbilitySystemComponent DamagedActor) "(GameplayTags=((TagName=\"SecretFacility.Invisibility.Active\")))")
  (Development|PrintString "goo on damaged"))

(event Custom|RemoveInvis
  (bind _playertagwaitmap (Variables|Default|GetPlayerTagWaitMap))
  (bind (_value _returnvalue) (Utilities|Map|Find _playertagwaitmap 0))
  ;[1 1 1] V2 -> Deprecated
  (GameplayEffects|RemoveActiveEffectswithGrantedTags (Ability|GetFortAbilitySystemComponent 0) "(GameplayTags=((TagName=\"SecretFacility.Invisibility.Active\")))")
  (if _returnvalue
    (Ability|Async|EndAction _value)
    (Utilities|Map|Remove _playertagwaitmap 0)))

(fn OnDamageEvent)

(fn ConstructionScript
  (bind _collisionbox (Variables|Default|GetCollisionBox))
  (Rendering|Components|Decal|SetDecalMaterial (Variables|Default|GetDecal) (Variables|Default|GetDecalMaterial))
  (Components|Box|SetBoxExtent _collisionbox (Math|Vector|MakeVector (Variables|Default|GetCollisionExtentX) (Variables|Default|GetCollisionExtentY) (ScalableFloat|GetValueasDouble (Variables|Default|GetHFCollisionBoxZExtent))) true)
  (Transformation|AddLocalOffset _collisionbox (Math|Vector|MakeVector (Variables|Default|GetCollisionOffsetX) (Variables|Default|GetCollisionOffsetY) (Variables|Default|GetCollisionOffsetZ)) false false)
  (Transformation|SetRelativeScale3D (Variables|Default|GetDecal) (Math|Vector|MakeVector (Variables|Default|GetDecalScaleX) (Variables|Default|GetDecalScaleY) (Variables|Default|GetDecalScaleZ)))
  (Transformation|AddLocalOffset (Variables|Default|GetDecal) (Math|Vector|MakeVector (Variables|Default|GetDecalOffsetX) (Variables|Default|GetDecalOffsetY) (Variables|Default|GetDecalOffsetZ)) false false))
