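; Automatic gas nozzle actor for Secret Facility that grants invisibility to pawns entering its gas.
; It stays on or toggles on a looping timer, switching collision and Niagara spawn parameters together.
; Overlap tracks pawns, clears any pending removal timer and applies the invisibility effect; exit starts a timer to remove it.
; One shared target variable and timer handle serve all pawns, so removal hits only the latest target.
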(event EventBeginPlay
  (if (== (Variables|Default|GetGasInactivePeriod) 0)
    (CallFunction|ActivateGas)
    (else
      (Utilities|Time|SetTimerbyEvent (delegate Custom|FlipActiveInactive) (Variables|Default|GetGasInactivePeriod) true false (Variables|Default|GetGasActiveStartDelay)))))

(event Custom|FlipActiveInactive
  (Utilities|FlowControl|FlipFlop
    (:A
      (CallFunction|ActivateGas))
    (:B
      (CallFunction|DeactivateGas))))

(event Collision|EventActorBeginOverlap
  (params (OtherActor :type Actor%20Object%20Reference))
  (bind _invisibilitygranthandle (Variables|Default|GetInvisibilityGrantHandle))
  (bind _currentpawnsingas (Variables|Default|GetCurrentPawnsInGas))
  (bind _output_get (Variables|Default|SetInvisibilityTarget OtherActor))
  (bind _asfort_pawn (Utilities|Casting|CastToFortPawn _output_get))
  (if (Actor|ActorHasTag (Variables|Default|GetTagInvisibilityGranted))
    _
    (elif (== (Utilities|Array|FindItem _currentpawnsingas _asfort_pawn) -1)
      (bind _returnvalue (Utilities|Array|AddUnique _currentpawnsingas _asfort_pawn))
      (Variables|Default|SetInvisibilityTarget (Utilities|Array|Get(acopy) (Variables|Default|GetCurrentPawnsInGas) _returnvalue))
      (if (Utilities|Time|IsValidTimerHandle _invisibilitygranthandle)
        (Utilities|Time|ClearandInvalidateTimerbyHandle _invisibilitygranthandle))
      (CallFunction|ApplyInvisibility))))

(event Collision|EventActorEndOverlap
  (params (OtherActor :type Actor%20Object%20Reference))
  (bind _output_get (Variables|Default|SetInvisibilityTarget OtherActor))
  (bind _asfort_pawn (Utilities|Casting|CastToFortPawn _output_get))
  (Utilities|Array|RemoveItem (Variables|Default|GetCurrentPawnsInGas) _asfort_pawn)
  (CallFunction|RemoveInvisibilityAfterDuration)
  (Development|PrintString "Pawn left gas"))

(event Custom|ApplyInvisibility
  (bind _invisibilitytarget (Variables|Default|GetInvisibilityTarget))
  (Fortnite|ApplyGameplayEffectsToActor _invisibilitytarget (Variables|Default|GetGEInvisibility) (Transformation|GetActorLocation))
  (Development|PrintString (Utilities|String|Append (Utilities|String|ToString(Object) _invisibilitytarget) " is invisible. Hopefully.") true true "(R=0.000000,G=0.660000,B=1.000000,A=1.000000)" 5.0))

(event Custom|RemoveInvisibilityAfterDuration
  (bind _returnvalue (Utilities|Time|SetTimerbyEvent (delegate Custom|RemoveInvisibility) (ScalableFloat|GetValueatLevel (Variables|Default|GetHFInvisibilityGrantDuration))))
  (Variables|Default|SetInvisibilityGrantHandle _returnvalue))

(event Custom|RemoveInvisibility
  (bind _invisibilitytarget (Variables|Default|GetInvisibilityTarget))
  (GameplayEffects|RemoveActiveGameplayEffectbySourceEffect (Ability|GetAbilitySystemComponent _invisibilitytarget) "/SecretFacilityGameplay/Gameplay/GameplayEffects/GE_SecretFacility_GrantInvisibilityInfinite.GE_SecretFacility_GrantInvisibilityInfinite_C" 0 -1)
  (Development|PrintString (Utilities|String|Append (Utilities|String|ToString(Object) _invisibilitytarget) " is no longer invisible. Hopefully.") true true "(R=0.000000,G=0.660000,B=1.000000,A=1.000000)" 5.0))

(fn ActivateGas
  (Collision|SetCollisionEnabled (Variables|Default|GetInvisibilityCollision) "QueryAndPhysics")
  (Niagara|SetNiagaraVariableByString(Float) (Variables|Default|GetNS_SecretFacility_InvisibilityGas_Jet) "User.MulitplierSmokeSpawnParticlesRate" 1.0)
  (Niagara|SetNiagaraVariableByString(Float) (Variables|Default|GetNS_SecretFacility_InvisibilityGas_Jet) "User.CoreSpawnParticlesProbability" 1.0)
  (Niagara|SetNiagaraVariableByString(Float) (Variables|Default|GetNS_SecretFacility_InvisibilityGas_Jet) "User.CloudSpawnParticlesRate" 1e+01))

(fn DeactivateGas
  (Collision|SetCollisionEnabled (Variables|Default|GetInvisibilityCollision) "NoCollision")
  (Niagara|SetNiagaraVariableByString(Float) (Variables|Default|GetNS_SecretFacility_InvisibilityGas_Jet) "User.MulitplierSmokeSpawnParticlesRate" 0.0)
  (Niagara|SetNiagaraVariableByString(Float) (Variables|Default|GetNS_SecretFacility_InvisibilityGas_Jet) "User.CoreSpawnParticlesProbability" 0.0)
  (Niagara|SetNiagaraVariableByString(Float) (Variables|Default|GetNS_SecretFacility_InvisibilityGas_Jet) "User.CloudSpawnParticlesRate" 0.0))

(fn ConstructionScript
  (Collision|SetCollisionEnabled (Variables|Default|GetInvisibilityCollision) "NoCollision"))
