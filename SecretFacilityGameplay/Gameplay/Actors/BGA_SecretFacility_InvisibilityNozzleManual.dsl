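; Secret Facility nozzle actor that emits an invisibility gas jet and toggles its InvisibilityCollision volume.
; ActivateGas enables query-and-physics collision and raises three Niagara user parameters for smoke, core and cloud spawning; DeactivateGas disables collision and zeroes them.
; The construction script starts with collision disabled.
; Separate custom Activate Gas and Deactivate Gas events leave collision untouched and both set the same "on" particle values.
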(event EventBeginPlay)

(event Custom|Activate%20Gas
  (Niagara|SetNiagaraVariableByString(Float) (Variables|Default|GetNS_SecretFacility_InvisibilityGas_Jet) "User.MulitplierSmokeSpawnParticlesRate" 1.0)
  (Niagara|SetNiagaraVariableByString(Float) (Variables|Default|GetNS_SecretFacility_InvisibilityGas_Jet) "User.CoreSpawnParticlesProbability" 1.0)
  (Niagara|SetNiagaraVariableByString(Float) (Variables|Default|GetNS_SecretFacility_InvisibilityGas_Jet) "User.CloudSpawnParticlesRate" 1e+01))

(event Custom|Deactivate%20Gas
  (Niagara|SetNiagaraVariableByString(Float) (Variables|Default|GetNS_SecretFacility_InvisibilityGas_Jet) "User.MulitplierSmokeSpawnParticlesRate" 1.0)
  (Niagara|SetNiagaraVariableByString(Float) (Variables|Default|GetNS_SecretFacility_InvisibilityGas_Jet) "User.CoreSpawnParticlesProbability" 1.0)
  (Niagara|SetNiagaraVariableByString(Float) (Variables|Default|GetNS_SecretFacility_InvisibilityGas_Jet) "User.CloudSpawnParticlesRate" 1e+01))

(fn ActivateGas
  (Collision|SetCollisionEnabled (Variables|Default|GetInvisibilityCollision) "QueryAndPhysics")
  (Niagara|SetNiagaraVariableByString(Float) (Variables|Default|GetNS_SecretFacility_InvisibilityGas_Jet) "User.MulitplierSmokeSpawnParticlesRate" 1.0)
  (Niagara|SetNiagaraVariableByString(Float) (Variables|Default|GetNS_SecretFacility_InvisibilityGas_Jet) "User.CoreSpawnParticlesProbability" 1.0)
  (Niagara|SetNiagaraVariableByString(Float) (Variables|Default|GetNS_SecretFacility_InvisibilityGas_Jet) "User.CloudSpawnParticlesRate" 1e+01))

(fn DeactivateGas
  (Collision|SetCollisionEnabled (Variables|Default|GetInvisibilityCollision) "NoCollision")
  (Niagara|SetNiagaraVariableByString(Float) (Variables|Default|GetNS_SecretFacility_InvisibilityGas_Jet) "User.MulitplierSmokeSpawnParticlesRate" 0.0)
  (Niagara|SetNiagaraVariableByString(Float) (Variables|Default|GetNS_SecretFacility_InvisibilityGas_Jet) "User.CoreSpawnParticlesProbability" 0.0)
  (Niagara|SetNiagaraVariableByString(Float) (Variables|Default|GetNS_SecretFacility_InvisibilityGas_Jet) "User.CloudSpawnParticlesRate" 0.0))

(fn ConstructionScript
  (Collision|SetCollisionEnabled (Variables|Default|GetInvisibilityCollision) "NoCollision"))
