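; Secret Facility actor that plays a burst effect marking the origin of an invisibility effect.
; Its custom PlayInvisibilityFX event executes a burst Gameplay Cue on itself, using the InvisibilityExplosion variable and passing the actor's location in the cue parameters.
; BeginPlay, overlap, Tick and the construction script are empty, and nothing in the graph calls the custom event.
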(event EventBeginPlay)

(event Collision|EventActorBeginOverlap
  (params (OtherActor :type Actor%20Object%20Reference)))

(event EventTick
  (params (DeltaSeconds :type Float%20%28single-precision%29)))

(event Custom|PlayInvisibilityFX
  (GameplayCue|ExecuteGameplayCueOnActor(Burst) self (Variables|Default|GetInvisibilityExplosion) (Ability|GameplayCue|MakeGameplayCueParameters :NormalizedMagnitude 0.0 :RawMagnitude 0.0 :Location (Transformation|GetActorLocation self) :Normal "0,0,0" :GameplayEffectLevel 1 :AbilityLevel 1 :bReplicateLocationWhenUsingMinimalRepProxy false)))

(fn ConstructionScript)
