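; Secret facility vault actor that spawns a loot pickup at its own location, serving as a weapon spawner.
; On BeginPlay it waits for the playlist data to be ready, then spawns the pickup from the secret facility loot tier group.
; The pickup is placed without a toss, and the overlap, tick and construction script hooks are empty.
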(event EventBeginPlay
  (Class|FortAsyncActionCurrentPlaylistReady|PlaylistDataReadyAsync
    (:OnReady
      (Inventory|SpawnPickupInWorldWithLootTier :LootTierName "LTG_SecretFacility" :Position (Transformation|GetActorLocation self) :OverrideMaxStackCount 0 :bToss false :bTossWithVelocity false :SourceType "Tossed" :Source "ItemSpawner" :LootWorldLevel 0))))

(event Collision|EventActorBeginOverlap
  (params (OtherActor :type Actor%20Object%20Reference)))

(event EventTick
  (params (DeltaSeconds :type Float%20%28single-precision%29)))

(fn ConstructionScript)
