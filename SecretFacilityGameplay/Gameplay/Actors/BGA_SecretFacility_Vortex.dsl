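; Secret Facility vortex actor whose rep-notifies drive a normal and a mega state.
; OnRep_Enabled toggles looping audio, collision and the vortex cylinder, and when enabling also reveals level-wide updraft glow actors and referenced point decals.
; OnRep_MegaVortexEnabled only handles turning on: it heightens the capsule and, unless already enabled, starts audio, shows the geyser cone, glows and decals, and spawns a light beam.
; The construction script sets the cone's relative rotation.
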(fn OnRep_Enabled
  (bind _enabled (Variables|Default|GetEnabled))
  (Components|Activation|Activate (Variables|Default|GetAudioShapeSpline) _enabled)
  (if _enabled
    (Audio|Components|Audio|Play (Variables|Default|GetLoopingAmbAudio) 0.0)
    (Collision|SetActorEnableCollisionwithChildren 0 true)
    (Rendering|SetVisibility (Variables|Default|GetVortexCylinder) true true)
    (Development|SetHiddeninGame (Variables|Default|GetVortexCylinder) false true)
    (Utilities|FlowControl|Sequence
      (:then_0
        (bind _outactors (Actor|GetAllActorsOfClass "/Game/Effects/Fort_Effects/Effects/Environment/Hera/SevenBase/BP_Env_SevenBase_Vortex_Updaraft_Glow.BP_Env_SevenBase_Vortex_Updaraft_Glow_C"))
        (for _array_element _outactors
          (Rendering|SetVisibility (Variables|/Game/Effects/Fort_Effects/Effects/Environment/Hera/SevenBase/BP_Env_SevenBase_Vortex_Updaraft_Glow.BP_Env_SevenBase_Vortex_Updaraft_Glow_C|GetDefaultSceneRoot _array_element) true true)))
      (:then_1
        (for _array_element_2 (Variables|Default|GetPointDecalActors)
          (Rendering|SetVisibility (Variables|/Game/TechArt/FakeLights/BP_PointDecal.BP_PointDecal_C|GetSM_PointDecal _array_element_2) true false))))
    (else
      (Audio|Components|Audio|Stop (Variables|Default|GetLoopingAmbAudio))
      (Components|Activation|Deactivate (Variables|Default|GetAudioShapeSpline))
      (Collision|SetActorEnableCollisionwithChildren 0)
      (Rendering|SetVisibility (Variables|Default|GetVortexCylinder) false true)
      (Development|SetHiddeninGame (Variables|Default|GetVortexCylinder) true true))))

(fn OnRep_MegaVortexEnabled
  (bind _enabled (Variables|Default|GetEnabled))
  (if (Variables|Default|GetMegaVortexEnabled)
    (Components|Capsule|SetCapsuleHalfHeight (Variables|Default|GetCapsule) 9e+03 true)
    (Collision|SetActorEnableCollisionwithChildren 0 true)
    (if (not _enabled)
      (Audio|Components|Audio|Play (Variables|Default|GetLoopingAmbAudio) 0.0)
      (Components|Activation|Activate (Variables|Default|GetAudioShapeSpline) _enabled)
      (Rendering|SetVisibility (Variables|Default|GetS_FX_GeyserCone) true true)
      (Development|SetHiddeninGame (Variables|Default|GetS_FX_GeyserCone) false true)
      (Game|SpawnActorfromClass "/SecretFacilityGameplay/Gameplay/Actors/BP_SecretFacility_LightBeam.BP_SecretFacility_LightBeam_C" (Transformation|GetActorTransform) "Undefined" "MultiplyWithRoot" 0 0)
      (Utilities|FlowControl|Sequence
        (:then_0
          (bind _outactors (Actor|GetAllActorsOfClass "/Game/Effects/Fort_Effects/Effects/Environment/Hera/SevenBase/BP_Env_SevenBase_Vortex_Updaraft_Glow.BP_Env_SevenBase_Vortex_Updaraft_Glow_C"))
          (for _array_element _outactors
            (Rendering|SetVisibility (Variables|/Game/Effects/Fort_Effects/Effects/Environment/Hera/SevenBase/BP_Env_SevenBase_Vortex_Updaraft_Glow.BP_Env_SevenBase_Vortex_Updaraft_Glow_C|GetDefaultSceneRoot _array_element) true true)))
        (:then_1
          (for _array_element_2 (Variables|Default|GetPointDecalActors)
            (Rendering|SetVisibility (Variables|/Game/TechArt/FakeLights/BP_PointDecal.BP_PointDecal_C|GetSM_PointDecal _array_element_2) true false)))))))

(fn ConstructionScript
  (|CallParentFunction)
  (Transformation|SetRelativeRotation (Variables|Default|GetS_FX_GeyserCone) (Math|Rotator|MakeRotator 0.0) false false))
