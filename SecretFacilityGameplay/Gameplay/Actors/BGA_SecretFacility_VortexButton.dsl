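; Interactable Fortnite building actor button that triggers the Secret Facility vortex escape sequence after a timed interaction.
; ActivateVortex fires every vortex's escape chute, opens all reverse Buckle Cabin doors, activates the referenced no-build zone, then clears bAllowInteract on all vortex buttons.
; Interaction overrides supply eligibility and prompt from variables.
; A separate Multi_MegaVortex_Button event destroys the button but is not called within this graph.
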(event EventBeginPlay)

(event Interaction|EventBlueprintOnInteract
  (params (InteractingPawn :type Fort%20Pawn%20Object%20Reference) (InteractionBeingAttempted :type EInteractionBeingAttempted%20Enum :default "FirstInteraction"))
  (CallFunction|ActivateVortex))

(event Custom|ActivateVortex
  (bind _outactors (Actor|GetAllActorsOfClass "/SecretFacilityGameplay/Gameplay/Actors/BGA_SecretFacility_Vortex.BGA_SecretFacility_Vortex_C"))
  (for _array_element _outactors
    (Class|BGASecretFacilityVortex|ActivateEscapeChute _array_element))
  (bind _outactors_2 (Actor|GetAllActorsOfClass "/DragonCartDoors/Gameplay/BuckleCabin/BP_BuckleCabin_Reverse_Ultra.BP_BuckleCabin_Reverse_Ultra_C"))
  (for _array_element_2 _outactors_2
    (Class|BPBuckleCabinReverseUltra|ManualDoorOpen _array_element_2))
  (Class|BSecretFacilityNoBuildZone|ActivateNoBuildZone (Variables|Default|GetRefNoBuildZone))
  (bind _outactors_3 (Actor|GetAllActorsOfClass "/SecretFacilityGameplay/Gameplay/Actors/BGA_SecretFacility_VortexButton.BGA_SecretFacility_VortexButton_C"))
  (for _array_element_3 _outactors_3
    (Variables|/Script/FortniteGame.BuildingActor|SetbAllowInteract false _array_element_3)))

(event Custom|Multi_MegaVortex_Button
  (Actor|DestroyActor))

(fn BlueprintCanInteract
  (params (InteractingPawn :type Fort%20Pawn%20Object%20Reference) (InteractionBeingAttempted :type EInteractionBeingAttempted%20Enum :default "FirstInteraction") (InteractionType :type TInteractionType%20Enum :default "IT_NoInteraction"))
  (return (Variables|Default|GetAllowInteraction)))

(fn BlueprintGetInteractionString
  (params (InteractingPawn :type Fort%20Pawn%20Object%20Reference) (InteractionBeingAttempted :type EInteractionBeingAttempted%20Enum :default "FirstInteraction"))
  (return (Variables|Default|GetInteractionName)))

(fn BlueprintGetInteractionTime
  (params (InteractingPawn :type Fort%20Pawn%20Object%20Reference) (InteractionBeingAttempted :type EInteractionBeingAttempted%20Enum :default "FirstInteraction"))
  (return 1.0 true))

(fn ConstructionScript)
