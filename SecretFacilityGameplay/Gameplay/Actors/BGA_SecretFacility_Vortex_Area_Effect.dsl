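; Secret Facility vortex area-effect actor that serves as an updraft escape chute, dormant until its ActivateEscapeChute event runs.
; One second after BeginPlay, a timer runs Ready, which stops audio, disables collision and hides the steam, cone, glow and Niagara updraft components.
; ActivateEscapeChute reverses this component by component and fires a burst GameplayCue alert at the actor's location.
; No launch logic is defined here, and Overlap and Tick are empty.
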(event Collision|EventActorBeginOverlap
  (params (OtherActor :type Actor%20Object%20Reference)))

(event EventTick
  (params (DeltaSeconds :type Float%20%28single-precision%29)))

(event EventBeginPlay
  (|CallParentFunction)
  (Utilities|Time|SetTimerbyEvent (delegate Custom|Ready) 1.0))

(event Custom|Ready
  ;[1 1 1] Turn off all visual/audio components at start (crude but had issues doing it in a more elegant way)
  (Audio|Components|Audio|Stop (Variables|Default|GetUpdraftAudio))
  (Audio|Components|Audio|Stop (Variables|Default|GetLoopingAmbAudio))
  (Collision|SetActorEnableCollisionwithChildren 0)
  (Rendering|SetVisibility (Variables|Default|GetP_GeyserSteam) false true)
  (Rendering|SetVisibility (Variables|Default|GetP_GeyserSteamBottom) false true)
  (Rendering|SetVisibility (Variables|Default|GetS_FX_GeyserCone) false false)
  (Rendering|SetVisibility (Variables|Default|GetSM_GlowDonutWind) false false)
  (Rendering|SetVisibility (Variables|Default|GetNS_Apollo_PowerPlant_Updraft_Converted) false false))

(event Custom|ActivateEscapeChute
  ;[1 1 1] Turn on all visual/audio components on activation (crude but had issues doing it in a more elegant way)
  (Audio|Components|Audio|Play (Variables|Default|GetUpdraftAudio) 0.0)
  (Audio|Components|Audio|Play (Variables|Default|GetLoopingAmbAudio) 0.0)
  (Collision|SetActorEnableCollisionwithChildren 0 true)
  (Rendering|SetVisibility (Variables|Default|GetS_FX_GeyserCone) true false)
  (Rendering|SetVisibility (Variables|Default|GetP_GeyserSteam) true true)
  (Rendering|SetVisibility (Variables|Default|GetP_GeyserSteamBottom) true true)
  (GameplayCue|ExecuteGameplayCueOnActor(Burst) self (Variables|Default|GetGCActivationAlert) (Ability|GameplayCue|MakeGameplayCueParameters :NormalizedMagnitude 0.0 :RawMagnitude 0.0 :Location (Transformation|GetActorLocation self) :Normal "0,0,0" :GameplayEffectLevel 1 :AbilityLevel 1 :bReplicateLocationWhenUsingMinimalRepProxy false))
  (Rendering|SetVisibility (Variables|Default|GetSM_GlowDonutWind) true false)
  (Rendering|SetVisibility (Variables|Default|GetNS_Apollo_PowerPlant_Updraft_Converted) true false))

(fn ConstructionScript
  (|CallParentFunction)
  (Transformation|SetRelativeRotation (Variables|Default|GetS_FX_GeyserCone) (Math|Rotator|MakeRotator 0.0) false false))
