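; Secret Facility Vortex no-build zone actor that tags its NoBuildZone component, presumably so building is blocked around the vortex.
; When the HF_NoBuildZoneEnableVortex scalable float reads true, BeginPlay copies the Component Tags variable onto that component.
; The ActivateNoBuildZone custom event repeats the same gated tag assignment on demand.
; Overlap, tick and construction script are empty, and no logic here clears the tags.
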(event EventBeginPlay
  (if (ScalableFloat|GetValueasBool (Variables|Default|GetHF_NoBuildZoneEnableVortex))
    (Variables|/Script/Engine.ActorComponent|SetComponentTags (Variables|Default|GetComponent%20Tags) (Variables|Default|GetNoBuildZone))))

(event Collision|EventActorBeginOverlap
  (params (OtherActor :type Actor%20Object%20Reference)))

(event EventTick
  (params (DeltaSeconds :type Float%20%28single-precision%29)))

(event Custom|ActivateNoBuildZone
  (if (ScalableFloat|GetValueasBool (Variables|Default|GetHF_NoBuildZoneEnableVortex))
    (Variables|/Script/Engine.ActorComponent|SetComponentTags (Variables|Default|GetComponent%20Tags) (Variables|Default|GetNoBuildZone))))

(fn ConstructionScript)
