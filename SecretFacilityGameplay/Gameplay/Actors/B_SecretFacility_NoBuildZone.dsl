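; Secret Facility actor that sets up a no-build zone and destroys builds targeted around it.
; On begin play or on ActivateNoBuildZone, it sets component tags on its NoBuildZone component, each path gated by its own hotfix scalable float.
; Activation then waits briefly, runs an async targeting preset, and applies build-destroying gameplay effects to each returned actor, passing the zone's location.
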(event EventBeginPlay
  (if (ScalableFloat|GetValueasBool (Variables|Default|GetHF_NoBuildZoneEnableStart))
    (Variables|/Script/Engine.ActorComponent|SetComponentTags (Variables|Default|GetComponent%20Tags) (Variables|Default|GetNoBuildZone))))

(event Collision|EventActorBeginOverlap
  (params (OtherActor :type Actor%20Object%20Reference)))

(event EventTick
  (params (DeltaSeconds :type Float%20%28single-precision%29)))

(event Custom|ActivateNoBuildZone
  (if (ScalableFloat|GetValueasBool (Variables|Default|GetHF_NoBuildZoneEnableVortex))
    (Variables|/Script/Engine.ActorComponent|SetComponentTags (Variables|Default|GetComponent%20Tags) (Variables|Default|GetNoBuildZone))
    (Utilities|FlowControl|Delay 0.1)
    (bind (_asynctaskproxy _targetinghandle) (Class|AsyncActionPerformTargeting|PerformTargetingAsyncAction 0 (Variables|Default|GetNoBuildEnvTargetingPreset))
      (:Targeted
        (for _array_element (TargetingSystem|TargetingResults|GetTargetingResultsActors (GameInstanceSubsystems|GetTargetingSubsystem) _targetinghandle)
          (Fortnite|ApplyGameplayEffectsToActor _array_element (Variables|Default|GetGEsDestroyBuilds) (Transformation|GetActorLocation self)))))))

(fn ConstructionScript)
